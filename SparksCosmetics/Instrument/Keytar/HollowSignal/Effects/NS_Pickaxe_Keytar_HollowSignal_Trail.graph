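# Niagara trail effect for the Hollow Signal keytar pickaxe, drawing short-lived animation trails from skeletal mesh bones or sockets.
# Two ribbon emitters, Trail_Color and Trail_Accent, spawn and update through the shared AnimTrail modules.
# Color curves fade each trail from an over-bright white through orange to red, with alpha falling off over life.
# Time-of-day scalars adjust opacity, Trail_Color also tints by time of day, and both feed dynamic material parameters.
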NiagaraSystem /SparksCosmetics/Instrument/Keytar/HollowSignal/Effects/NS_Pickaxe_Keytar_HollowSignal_Trail
dependencies {
 usedDataInterfaces [&NiagaraDataInterfaceColorCurve, &NiagaraDataInterfaceCurve]
 usedDynamicInputs []
 usedModules [
  &SystemState
  &self:SystemSpawnScript.SystemScriptSource.SystemScriptGraph.NiagaraNodeAssignment_0.SetVariables_#1
  &self:SystemSpawnScript.SystemScriptSource.SystemScriptGraph.NiagaraNodeAssignment_2.SetVariables_#2
  &self:SystemSpawnScript.SystemScriptSource.SystemScriptGraph.NiagaraNodeAssignment_1.SetVariables_#3
 ]
}
emitters {
 Trail_Accent {
  emitterUpdateScript [
   &EmitterState()
   &FN_SpawnAnimTrail_v2()
  ]
  particleSpawnScript [
   /Game/Effects/Niagara/Modules/AnimTrail/FN_InitializeAnimTrail/FN_InitializeAnimTrailBoneIndecies(First Bone or Socket Index=1)
   /Game/Effects/Niagara/Modules/AnimTrail/FN_InitializeAnimTrail/FN_InitializeAnimTrail(AnimTrailWidth=1, Color=<rgba 1 1 1 0.5>, Lifetime=0.15)
   &AddVelocity(Velocity=<-500 0 0>) {enabled=false}
  ]
  particleUpdateScript [
   &FN_UpdateAnimTrail_v2()
   &ScaleColor(
    Scale Alpha=&Multiply_Float(
     A=&FloatFromCurve(FloatCurve=DataInterface() {propertyValues {Curve {keys [
          {interpMode=RCIM_Cubic, leaveTangent=-3.4731, tangentMode=RCTM_Break, time=0, value=1}
          <at 0.6689 0.0233>
          {arriveTangent=-0.2139, interpMode=RCIM_Cubic, tangentMode=RCTM_Break, time=1, value=0}
         ]}}})
     B=&FNTimeOfDayScalar(Night Scale=0.65, Morning Scale=0.5, Day Scale=0.7, Evening Scale=0.7)
    )
    Scale RGB=&MakeVectorFromLinearColorRGB(Color=&ColorFromCurve(ColorCurve=DataInterface() {propertyValues {
        AlphaCurve {keys [<at 0 1 RCIM_Cubic>]}
        BlueCurve {keys [
          {arriveTangent=-20.4245, interpMode=RCIM_Cubic, leaveTangent=-20.4245, tangentMode=RCTM_User, time=0, value=5}
          {arriveTangentWeight=0.7193, interpMode=RCIM_Cubic, leaveTangentWeight=0.9651, tangentMode=RCTM_User, time=0.3343, value=0.026}
          <at 1 0.0466>
         ]}
        GreenCurve {keys [
          {arriveTangent=-20.4245, interpMode=RCIM_Cubic, leaveTangent=-20.4245, tangentMode=RCTM_User, time=0, value=5}
          {arriveTangentWeight=0.7258, interpMode=RCIM_Cubic, leaveTangentWeight=0.9739, tangentMode=RCTM_User, time=0.3343, value=0.2496}
          <at 1 0>
         ]}
        RedCurve {keys [
          {arriveTangent=-20.4245, interpMode=RCIM_Cubic, leaveTangent=-20.4245, tangentMode=RCTM_User, time=0, value=5}
          {arriveTangentWeight=0.5869, interpMode=RCIM_Cubic, leaveTangentWeight=0.7875, tangentMode=RCTM_User, time=0.3343, value=1}
          <at 1 1>
         ]}
       }}))
    ScaleA=true
    ScaleRGB=true
   )
   &SolveForcesAndVelocity(Clamp Velocity=false, Limit Acceleration=false)
   &DynamicMaterialParameters(
    Index 0 Param 4=0.9
    Index 1 Param 3=&FloatFromCurve(
     FloatCurve=DataInterface() {propertyValues {Curve {keys [<at 0 0>, {arriveTangent=1.9814, tangentMode=RCTM_User, time=1, value=1}]}}}
     CurveIndex=&ModuloFloat(A=$Emitter.Age, B=1)
     Scale Curve=0
    )
    Index 1 Param 4=1
    Index 2 Param 1=10
    Index 2 Param 2=1
    Index 2 Param 3=0.5
    Index 2 Param 4=0.6
    Module.Index0Float0Write=true
    Module.Index0Float1Write=true
    Module.Index0Float2Write=true
    Module.Index0Float3Write=true
    Module.Index1Float0Write=true
    Module.Index1Float1Write=true
    Module.Index1Float2Write=true
    Module.Index1Float3Write=true
    Module.Index2Float0Write=true
    Module.Index2Float1Write=true
    Module.Index2Float2Write=true
    Module.Index2Float3Write=true
    Write Parameter Index 1=::NewEnumerator1
    Write Parameter Index 2=::NewEnumerator1
   )
  ]
  rendererClasses [&NiagaraRibbonRendererProperties]
  renderers [{rendererClass=&NiagaraRibbonRendererProperties}]
 }
 Trail_Color {
  emitterUpdateScript [
   &EmitterState()
   &FN_SpawnAnimTrail_v2()
  ]
  particleSpawnScript [
   &FN_InitializeAnimTrailBoneIndecies(First Bone or Socket Index=1)
   &FN_InitializeAnimTrail(AnimTrailWidth=1, Color=<rgba 1 1 1 0.05>, Lifetime=0.2)
   &AddVelocity(Velocity=<-500 0 0>) {enabled=false}
  ]
  particleUpdateScript [
   &ScaleColor(
    Scale Alpha=&FloatFromCurve(
     FloatCurve=DataInterface() {propertyValues {Curve {keys [
         {interpMode=RCIM_Cubic, leaveTangent=-2.3646, tangentMode=RCTM_Break, time=0, value=1}
         <at 0.6689 0.0233>
         {arriveTangent=-0.2139, interpMode=RCIM_Cubic, tangentMode=RCTM_Break, time=1, value=0}
        ]}}}
     Scale Curve=&FNTimeOfDayScalar(Night Scale=0.7, Morning Scale=0.9, Evening Scale=1)
    )
    Scale RGB=&Multiply_Vector(
     A=&MakeVectorFromLinearColorRGB(Color=&ColorFromCurve(ColorCurve=DataInterface() {propertyValues {
         AlphaCurve {keys [<at 0 1 RCIM_Cubic>]}
         BlueCurve {keys [
           {arriveTangent=-20.4245, interpMode=RCIM_Cubic, leaveTangent=-20.4245, tangentMode=RCTM_User, time=0, value=5}
           {arriveTangentWeight=0.7193, interpMode=RCIM_Cubic, leaveTangentWeight=0.9651, tangentMode=RCTM_User, time=0.3343, value=0.026}
           <at 1 0.0466>
          ]}
         GreenCurve {keys [
           {arriveTangent=-20.4245, interpMode=RCIM_Cubic, leaveTangent=-20.4245, tangentMode=RCTM_User, time=0, value=5}
           {arriveTangentWeight=0.7258, interpMode=RCIM_Cubic, leaveTangentWeight=0.9739, tangentMode=RCTM_User, time=0.3343, value=0.2496}
           <at 1 0>
          ]}
         RedCurve {keys [
           {arriveTangent=-20.4245, interpMode=RCIM_Cubic, leaveTangent=-20.4245, tangentMode=RCTM_User, time=0, value=5}
           {arriveTangentWeight=0.5869, interpMode=RCIM_Cubic, leaveTangentWeight=0.7875, tangentMode=RCTM_User, time=0.3343, value=1}
           <at 1 1>
          ]}
        }}))
     B=&FN_Select_Vector_from_Time_Of_Day(Morning=<0.9 0.9 0.9>, Day=<1 1 1>, Evening=<1 1 1>, Night=<0.5 0.5 1>)
    )
    ScaleA=true
    ScaleRGB=true
   ) {module=ScaleColor002}
   &FN_UpdateAnimTrail_v2()
   &SolveForcesAndVelocity(Clamp Velocity=false, Limit Acceleration=false)
   &DynamicMaterialParameters(
    Index 0 Param 1=1
    Index 1 Param 3=&FloatFromCurve(
     FloatCurve=DataInterface() {propertyValues {Curve {keys [<at 0 0>, {arriveTangent=1.9814, tangentMode=RCTM_User, time=1, value=1}]}}}
     CurveIndex=&ModuloFloat(A=$Emitter.Age, B=1)
     Scale Curve=0
    )
    Index 1 Param 4=1
    Index 2 Param 1=10
    Index 2 Param 2=1
    Index 2 Param 3=0.5
    Index 2 Param 4=0.6
    Module.Index0Float0Write=true
    Module.Index0Float1Write=true
    Module.Index0Float2Write=true
    Module.Index0Float3Write=true
    Module.Index1Float0Write=true
    Module.Index1Float1Write=true
    Module.Index1Float2Write=true
    Module.Index1Float3Write=true
    Module.Index2Float0Write=true
    Module.Index2Float1Write=true
    Module.Index2Float2Write=true
    Module.Index2Float3Write=true
    Write Parameter Index 1=::NewEnumerator1
    Write Parameter Index 2=::NewEnumerator1
   )
  ]
  rendererClasses [&NiagaraRibbonRendererProperties]
  renderers [{rendererClass=&NiagaraRibbonRendererProperties}]
 }
}
summary {
 systemName=NS_Pickaxe_Keytar_HollowSignal_Trail
 emitters [
  {emitterName=Trail_Color, rendererClasses [&NiagaraRibbonRendererProperties]}
  {emitterName=Trail_Accent, rendererClasses [&NiagaraRibbonRendererProperties]}
 ]
 userVariables [
  User.bIsFrontEndPreview {type=NiagaraBool}
  User.RibbonIDOffset {type=NiagaraInt32}
  User.SkeletalMesh {
   type=NiagaraDataInterfaceSkeletalMesh
   defaultValue {
    struct=NiagaraExt_VariableValue_DataInterface
    value {dataInterface=&self:NiagaraDataInterfaceSkeletalMesh_0}
   }
  }
 ]
}
